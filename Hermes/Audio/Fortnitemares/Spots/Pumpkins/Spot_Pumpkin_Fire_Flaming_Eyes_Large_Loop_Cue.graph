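# Looping ambient Sound Cue for large flaming-eye pumpkin fire props in Fortnitemares.
# A Mixer layers a stereo pumpkin fire loop with a reused fire urn loop.
# Two endlessly looping Envelopers of different lengths, one shaping pitch and one volume, add slow drift before a Modulator randomizes pitch slightly.
# Inline overrides set logarithmic attenuation with distance filtering, listener focus and occlusion, plus a small per-owner concurrency limit.
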SoundCue /Hermes/Audio/Fortnitemares/Spots/Pumpkins/Spot_Pumpkin_Fire_Flaming_Eyes_Large_Loop_Cue
root=Modulator_2
nodes [
 {
  id=Enveloper_0
  children [Mixer_0]
  properties {
   bLoop=true
   bLoopIndefinitely=true
   loopEnd=30
   pitchCurve {editorCurveData {keys [{interpMode=RCIM_Cubic, time=0}, <at 15.9468 1.1 RCIM_Cubic>, {interpMode=RCIM_Cubic, time=30}]}}
   volumeCurve {editorCurveData {keys [<at 0 1>]}}
  }
 }
 {
  id=Enveloper_3
  children [Enveloper_0]
  properties {
   bLoop=true
   bLoopIndefinitely=true
   loopEnd=20
   pitchCurve {editorCurveData {keys [{time=0}]}}
   volumeCurve {editorCurveData {keys [
      {arriveTangent=-0.01, interpMode=RCIM_Cubic, time=0, value=1}
      <at 10 1.1 RCIM_Cubic>
      {interpMode=RCIM_Cubic, leaveTangent=0.01, time=20, value=1}
     ]}}
  }
 }
 {id=Mixer_0, children [WavePlayer_1,WavePlayer_0], properties {inputVolume [0.8,1.1]}}
 {id=Modulator_2, children [Enveloper_3], properties {pitchMax=1.1, pitchMin=1, volumeMax=1, volumeMin=1}}
 {
  id=WavePlayer_0
  properties {bLooping=true, soundWaveAssetPtr=&/Spot_Pumpkin_Fire_Flaming_Eyes_Large_Loop_Stereo}
 }
 {id=WavePlayer_1, properties {bLooping=true, soundWaveAssetPtr=/AmbientAudioContent/Waves/SpottedLoops/Spot_Fire_Urn_01}}
]
properties {
 bOverrideAttenuation=true
 bOverrideConcurrency=true
 maxDistance=1150
 soundClassObject=&Fort_Ambient_Props_Athena
 volumeMultiplier=0.6
 attenuationOverrides {
  bAttenuateWithLPF=true
  bEnableListenerFocus=true
  distanceAlgorithm=Logarithmic
  falloffDistance=950
  focusDistanceScale=0.9
  hPFFrequencyAtMax=400
  hPFFrequencyAtMin=200
  lPFRadiusMax=1000
  lPFRadiusMin=500
  nonFocusVolumeAttenuation=0.4
  occlusionInterpolationTime=0.4
  occlusionLowPassFilterFrequency=900
  occlusionVolumeAttenuation=0.5
  reverbDistanceMin=0
  stereoSpread=100
  attenuationShapeExtents=<200 64.5942 90>
  customAttenuationCurve {editorCurveData {keys [{arriveTangent=-2.3812, interpMode=RCIM_Cubic, leaveTangent=-2.3812, tangentMode=RCTM_User, time=0, value=1}, <at 1 0.01 RCIM_Cubic>]}}
 }
 concurrencyOverrides {bLimitToOwner=true, bVolumeScaleCanRelease=true, maxCount=2}
}
