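# Looping ambient Sound Cue for the small flaming-eyes pumpkin fire in Fortnitemares, on the Athena ambient props sound class.
# A Mixer layers a quieter generic torch loop under a dedicated mono pumpkin fire loop.
# The root Modulator boosts volume by a fixed amount and slightly randomizes pitch upward.
# Inline overrides set logarithmic attenuation with listener focus and cap concurrency at two instances per owner.
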SoundCue /Hermes/Audio/Fortnitemares/Spots/Pumpkins/Spot_Pumpkin_Fire_Flaming_Eyes_Small_Loop_Cue
root=Modulator_2
nodes [
 {id=Mixer_0, children [WavePlayer_3,WavePlayer_0], properties {inputVolume [0.4,1]}}
 {id=Modulator_2, children [Mixer_0], properties {pitchMax=1.1, pitchMin=1, volumeMax=2, volumeMin=2}}
 {
  id=WavePlayer_0
  properties {bLooping=true, soundWaveAssetPtr=&/Spot_Pumpkin_Fire_Flaming_Eyes_Small_Loop_Mono}
 }
 {id=WavePlayer_3, properties {bLooping=true, soundWaveAssetPtr=/AmbientAudioContent/Waves/SpottedLoops/Spot_Fire_Torch_03}}
]
properties {
 bOverrideAttenuation=true
 bOverrideConcurrency=true
 maxDistance=1250
 soundClassObject=&Fort_Ambient_Props_Athena
 volumeMultiplier=1.5
 attenuationOverrides {
  bEnableListenerFocus=true
  distanceAlgorithm=Logarithmic
  falloffDistance=950
  focusDistanceScale=0.9
  nonFocusVolumeAttenuation=0.4
  occlusionInterpolationTime=0.4
  occlusionLowPassFilterFrequency=900
  occlusionVolumeAttenuation=0.5
  reverbDistanceMin=0
  stereoSpread=100
  attenuationShapeExtents=<300 64.5942 90>
  customAttenuationCurve {editorCurveData {keys [{arriveTangent=-2.3812, interpMode=RCIM_Cubic, leaveTangent=-2.3812, tangentMode=RCTM_User, time=0, value=1}, <at 1 0.01 RCIM_Cubic>]}}
 }
 concurrencyOverrides {bLimitToOwner=true, bVolumeScaleCanRelease=true, maxCount=2}
}
